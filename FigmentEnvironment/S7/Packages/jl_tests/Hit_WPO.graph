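# Material function that computes a hit-bounce world position offset by scaling a precomputed hit vector with distance falloff from a deformation center.
# Distance times WSDamageRadius, clamped and inverted, forms the falloff, with an optional static-switched inversion blend.
# Mobile feature levels use Custom HLSL that branches to zero beyond a camera-distance threshold and ignores that inversion.
# A Gradient output gives the falloff, or an object-radius gradient from mesh center for crack shading.
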MaterialFunction /FigmentEnvironment/S7/Packages/jl_tests/Hit_WPO
graph {
 properties {libraryCategoriesText [Misc,Fortnite]}
 nodes [
  MaterialExpressionObjectWorldPosition_8 = ObjectPositionWS()
  n1 = FunctionInput(Preview=@MaterialExpressionObjectWorldPosition_8) {bUsePreviewValueAsDefault=true, inputName=WSDeformationCenter, sortPriority=10}
  n2 = FunctionInput(Preview=WorldPosition().XYZ) {bUsePreviewValueAsDefault=true, inputName=WorldPosition}
  n3 = FunctionInput() {bUsePreviewValueAsDefault=true, inputName=WSDamageRadius, inputType=FunctionInput_Scalar, sortPriority=10, previewValue=<1 0 0 1>}
  n4 = Multiply(A=Distance(A=@n1, B=@n2), B=@n3)
  Clamp(None=@n4)
  n5 = CollectionParameter() {
   collection=/FigmentEnvironment/S7/Packages/Fortress_SharedMaterials/GlobalMaterialParameters/FortniteMaterialParameters
   parameterName=ImpactBounceStrength
  }
  ComponentMask(None=@n5) {r=true}
  Custom!(None) {code="float result = 1.0;\r\nif (first >= second)\r\n{\r\nresult = 0.0;\r\n}\r\nreturn result;"}
  FunctionInput() {bUsePreviewValueAsDefault=true, description="0-1 ", inputName=DestructionAnimation, inputType=FunctionInput_Scalar, previewValue=<0.5 0 0 1>}
  FunctionInput() {bUsePreviewValueAsDefault=true, description=Magnitude controls bounce amount, inputName=Code_DeformationVector, previewValue=<0 1 0 1>}
  FunctionInput() {bUsePreviewValueAsDefault=true, description="0-1 ", inputName=DestructionAmount, inputType=FunctionInput_Scalar}
  MaterialExpressionFunctionInput_6 = FunctionInput(Preview=VectorParameter() {group=z_code, parameterName=Code_PrecomputedHitVector, defaultValue=<rgba 0 0 1 0>}.RGB) {bUsePreviewValueAsDefault=true, description=Magnitude controls bounce amount, inputName=Code_PrecomputedHitVector, previewValue=<0 1 0 1>}
  n6 = Clamp(None=@n4)
  n7 = OneMinus(None=@n6)
  n8 = FunctionInput() {bUsePreviewValueAsDefault=true, description=1 is inverted, inputName=InvertDamageFalloff, inputType=FunctionInput_Scalar, sortPriority=21}
  n9 = FunctionInput(Preview=StaticBool()) {
   bUsePreviewValueAsDefault=true
   inputName=ExposeLerpFunctionForInvertingTheHitDistanceCalc
   inputType=FunctionInput_StaticBool
   sortPriority=20
  }
  n10 = Multiply(A=@MaterialExpressionFunctionInput_6, B=StaticSwitch(True=LinearInterpolate(A=@n7, B=@n6, Alpha=@n8), False=@n7, Value=@n9))
  MaterialExpressionScalarParameter_3 = ScalarParameter() {defaultValue=1300000, group=z_code, parameterName=MaxDistSqr}
  MaterialExpressionReroute_3 = Reroute(None=@n3)
  MaterialExpressionObjectPositionWS_5 = ObjectPositionWS()
  MaterialExpressionCameraPositionWS_4 = CameraPositionWS()
  MaterialExpressionReroute_1 = Reroute(None=@n1)
  n11 = Custom!(
   MaxDistSqr=@MaterialExpressionScalarParameter_3
   DamageRadius=@MaterialExpressionReroute_3
   DeformationVector=@MaterialExpressionFunctionInput_6
   ObjPos=@MaterialExpressionObjectPositionWS_5
   CameraPos=@MaterialExpressionCameraPositionWS_4
   HitPos=@MaterialExpressionReroute_1
  ) {
   code="float3 ObjToCamera = (ObjPos - CameraPos);\r\nfloat DistToObj = dot(ObjToCamera, ObjToCamera);\r\nfloat3 result = (0.0,0.0,0.0);\r\n\r\nBRANCH\r\nif (DistToObj < MaxDistSqr)\r\n{\r\nresult = (DeformationVector * (1.0 - (saturate((DamageRadius * (distance(HitPos, WSDemote(GetWorldPosition(Parameters)))))))));\r\n}\r\nreturn result;"
   desc="\n*** Original source before expression upgrade ***\nfloat3 ObjToCamera = (ObjPos - CameraPos);\r\nfloat DistToObj = dot(ObjToCamera, ObjToCamera);\r\nfloat3 result = (0.0,0.0,0.0);\r\n\r\nBRANCH\r\nif (DistToObj < MaxDistSqr)\r\n{\r\nresult = (DeformationVector * (1.0 - (saturate((DamageRadius * (distance(HitPos, GetWorldPosition(Parameters))))))));\r\n}\r\nreturn result;"
   inputs [
    {inputName=MaxDistSqr, input {expression=&self:MaterialExpressionScalarParameter_3}}
    {inputName=DamageRadius, input {expression=&self:MaterialExpressionReroute_3}}
    {inputName=DeformationVector, input {expression=&self:MaterialExpressionFunctionInput_6}}
    {inputName=ObjPos, input {expression=&self:MaterialExpressionObjectPositionWS_5}}
    {inputName=CameraPos, input {expression=&self:MaterialExpressionCameraPositionWS_4}}
    {inputName=HitPos, input {expression=&self:MaterialExpressionReroute_1}}
   ]
  }
  FunctionOutput(None=FeatureLevelSwitch(Default=@n10, ES2=@n11, ES3_1=@n11, SM4_REMOVED=@n10))
  n12 = ObjectRadius()
  n13 = Constant() {r=0.8}
  n14 = Clamp(None=Divide(A=Distance(A=@MaterialExpressionObjectWorldPosition_8, B=@n2), B=Multiply(A=@n12, B=@n13)))
  n15 = StaticBool()
  FunctionOutput(None=StaticSwitch(
    True=OneMinus(None=@n14)
    False=@n7
    Value=FunctionInput(Preview=@n15) {bUsePreviewValueAsDefault=true, inputName=CrackDamageShader, inputType=FunctionInput_StaticBool}
   ) {desc=of the crack shader display cracks from center of mesh}) {outputName=Gradient}
  Subtract(A=@n7) {desc=Fix One unit offset of all meshes}
 ]
}
