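# Shared material function that adds Fortnite hit feedback, crack darkening and vertex deformation, to damaged surfaces.
# Hit_WPO turns code-driven damage inputs into a deformation offset and gradient, with the offset also feeding LODEffect for distance-cull shrink WPO.
# Diffuse cracks height-lerp by destruction amount, sampled from UV1 or world-aligned projection, with a cheaper blend for mobile and low/medium quality.
# Emissive passes through; the gradient-based crack clamp is unconnected.
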MaterialFunction /FigmentEnvironment/S7/Packages/jl_tests/Fortnite_HitEffects
graph {nodes [
  n1 = FunctionInput() {
   bUsePreviewValueAsDefault=true
   inputName=ScalarOffsetforDamageCracks
   inputType=FunctionInput_Scalar
   sortPriority=11
   previewValue=<0.1 0 0 1>
  }
  n2 = FunctionInput() {bUsePreviewValueAsDefault=true, inputName=ScalarModDamageCracks, inputType=FunctionInput_Scalar, sortPriority=10, previewValue=<1 0 0 1>}
  n3 = FunctionInput(Preview=ScalarParameter() {group=z_code, parameterName=Code_DestructionAmount}) {bUsePreviewValueAsDefault=true, inputName=Code_DestructionAmount, inputType=FunctionInput_Scalar, previewValue=<0 1 0 1>}
  n4 = FunctionInput(Preview=WorldPosition().XYZ) {bUsePreviewValueAsDefault=true, inputName=WorldPosition, sortPriority=200}
  n5 = FunctionInput(Preview=StaticBool()) {
   bUsePreviewValueAsDefault=true
   inputName=ExposeLerpFunctionForInvertingTheHitDistanceCalc
   inputType=FunctionInput_StaticBool
   sortPriority=20
  }
  n6 = FunctionInput() {bUsePreviewValueAsDefault=true, description=1 is inverted, inputName=InvertDamageFalloff, inputType=FunctionInput_Scalar, sortPriority=21}
  n7 = MaterialFunctionCall!(
   CrackDamageShader=FunctionInput(Preview=StaticBool()) {bUsePreviewValueAsDefault=true, inputName=CrackDamageShader, inputType=FunctionInput_StaticBool}
   DestructionAmount=@n3
   Code_DeformationVector=FunctionInput(Preview=VectorParameter() {group=z_code, parameterName=Code_DeformationVector, defaultValue=<rgba 0 0 1 0>}.RGB) {bUsePreviewValueAsDefault=true, inputName=Code_DeformationVector, previewValue=<0 1 0 1>}
   WorldPosition=@n4
   DestructionAnimation=FunctionInput(Preview=ScalarParameter() {group=z_code, parameterName=Code_HitAnimation}) {bUsePreviewValueAsDefault=true, inputName=Code_HitAnimation, inputType=FunctionInput_Scalar}
   Code_PrecomputedHitVector
   WSDamageRadius=FunctionInput(Preview=ScalarParameter() {defaultValue=1, group=z_code, parameterName=WSDamageRadius}) {bUsePreviewValueAsDefault=true, inputName=WSDamageRadius, inputType=FunctionInput_Scalar}
   WSDeformationCenter=FunctionInput(Preview=VectorParameter() {group=z_code, parameterName=WSDeformationCenter}.RGB) {bUsePreviewValueAsDefault=true, inputName=WSDeformationCenter}
   ExposeLerpFunctionForInvertingTheHitDistanceCalc=@n5
   InvertDamageFalloff=@n6
  ) {materialFunction=&/Hit_WPO}
  n8 = Multiply(A=Multiply(A=@n3, B=Constant() {r=1.5}), B=@n7.Gradient)
  Clamp(None=Add(A=@n1, B=Multiply(A=@n2, B=@n8))) {maxDefault=0.6}
  n9 = FunctionInput(Preview=TextureSample() {texture=/FigmentEnvironment/S7/Packages/Fortress_Walls_Generic/Textures/T_FORT_Plaster_A_Cartoon}.RGB) {bUsePreviewValueAsDefault=true, inputName=Diffuse}
  n10 = StaticSwitch(
   True=TextureSample(UVs=TextureCoordinate() {coordinateIndex=1}) {texture=&/Cracks_02}.RGB
   False=MaterialFunctionCall(
    TextureObject=TextureObject() {texture=&Cracks_02}
    TextureSize=Constant() {r=350}
    WorldPosition=WorldPosition() {worldPositionShaderOffset=WPT_ExcludeAllShaderOffsets}.XYZ
   ) {materialFunction=&WorldAlignedTexture}.XYZ Texture
   Value=FunctionInput(Preview=StaticBool()) {bUsePreviewValueAsDefault=true, inputName=Code_UseUvForDamage, inputType=FunctionInput_StaticBool}
  ) {desc=use uv 2 for destruction - save 40 instructions}
  n11 = MaterialFunctionCall(Transition Phase=@n3, Height Texture=ComponentMask(None=@n10) {b=true}, Contrast=Constant() {r=2}) {materialFunction=&HeightLerp}
  n12 = ComponentMask(None=@n10) {r=true}
  n13 = Multiply(A=@n11.Alpha, B=@n12)
  n14 = Add(A=Multiply(A=@n9, B=OneMinus(None=@n13)) {desc=*diffuse}, B=Multiply(A=@n11.Alpha, B=ComponentMask(None=@n10) {g=true})) {desc=69 - 26 with uv coordinates}
  n15 = Multiply(A=@n9, B=LinearInterpolate(A=Constant() {r=1}, B=OneMinus(None=@n12), Alpha=@n3))
  n16 = FeatureLevelSwitch(Default=@n14, ES2=@n15, ES3_1=@n15, SM4_REMOVED=@n14)
  n17 = FunctionInput(Preview=StaticBoolParameter() {defaultValue=true, group=z_code, parameterName=Code_UseHitEffects}) {bUsePreviewValueAsDefault=true, inputName=Code_UseHitEffects, inputType=FunctionInput_StaticBool}
  FunctionOutput(None=StaticSwitch(True=QualitySwitch(Default=@n16, Low=@n15, Medium=@n15), False=@n9, Value=@n17)) {outputName=Diffuse, sortPriority=1}
  n18 = StaticSwitch(True=@n7.Result, False=Constant(), Value=@n17)
  n19 = FunctionInput(Preview=ScalarParameter() {group=WorldPositionOffset, parameterName=DistanceCullFadeAnimationOffset}) {bUsePreviewValueAsDefault=true, inputName=DistanceCullFadeAnimationOffset, inputType=FunctionInput_Scalar, sortPriority=1000}
  n20 = MaterialFunctionCall!(World Position=Add(A=@n18, B=@n4), DistanceCullFadeAnimationOffset=@n19, Use actor center instead of pivot point) {
   desc=Contains an interal feature level switch
   materialFunction=/FigmentEnvironment/S7/Packages/Fortress_SharedMaterials/Base_Material_Functions/LODEffect
  }
  FunctionOutput(None=Add(A=@n18, B=@n20.Result) {constB=0}) {outputName=Hit And Shrink WPO, sortPriority=3}
  FunctionOutput(None=FunctionInput() {bUsePreviewValueAsDefault=true, inputName=Code_Emissive}) {outputName=Emissive, sortPriority=2}
  FunctionOutput(None=@n18) {outputName=Hit WPO, sortPriority=4}
  FunctionOutput(None=@n7.Gradient) {outputName=Hit_Gradient, sortPriority=6}
  FunctionOutput(None=@n20.Result) {outputName=Shrink WPO, sortPriority=5}
 ]}
